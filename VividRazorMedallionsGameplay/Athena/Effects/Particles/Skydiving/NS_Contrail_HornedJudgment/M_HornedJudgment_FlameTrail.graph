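# Translucent unlit two-sided Niagara material for the HornedJudgment skydiving contrail's flame trail, usable on ribbons, sprites or meshes.
# A panned, noise-distorted flame texture and procedural falloff masks build shape, three-color ramp and opacity, softened by dissolve, depth, camera and optional fresnel fades.
# Dynamic parameters drive UV offset, dive emissive boost, hue shift, dissolve, bend and camera offset.
# Vertex offset adds wind wobble, with optional world-wind noise and spline thickening.
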Material /VividRazorMedallionsGameplay/Athena/Effects/Particles/Skydiving/NS_Contrail_HornedJudgment/M_HornedJudgment_FlameTrail
graph {
 properties {
  bUsedWithInstancedStaticMeshes=true
  bUsedWithNiagaraMeshParticles=true
  bUsedWithNiagaraRibbons=true
  bUsedWithNiagaraSprites=true
  blendMode=BLEND_Translucent
  pixelDepthOffsetMode=PDOM_Legacy
  shadingModel=MSM_Unlit
  twoSided=true
  usedShadingModels=Unlit
 }
 nodes [
  n1 = MaterialFunctionCall() {materialFunction=&CameraDirectionVector}
  n2 = DynamicParameter() {defaultValue=<rgba 0.5 0 0 0>, paramNames [UV1 U Offset, Dive, Camera Offset, HueShift]}
  n3 = SmoothStep(
   Min=ScalarParameter() {group=07- WPO, parameterName=WindSmoothMin, sortPriority=3}
   Max=ScalarParameter() {defaultValue=0.25, group=07- WPO, parameterName=WindSmoothMax, sortPriority=4}
   Value=ComponentMask(None=NamedRerouteUsage()) {r=true}
  )
  n4 = NamedRerouteUsage()
  n5 = Multiply(A=NamedRerouteUsage(), B=ScalarParameter() {defaultValue=0.5, group=07- WPO, parameterName=Wind_DistoStrength, sortPriority=2})
  n6 = VectorParameter() {
   group=07- WPO
   parameterName=Wind_UV_Settings
   sortPriority=0
   channelNames {a=PanY, b=PanX, g=ScaleY, r=ScaleX}
   defaultValue=<rgba 1 1 -1 0>
  }
  n7 = Panner(
   Coordinate=Add(
    A=DynamicParameter() {defaultValue=<rgba 0.5 0 0 0>, paramNames [UV1 U Offset, Dive, Camera Offset, HueShift]}.UV1 U Offset
    B=Add(A=@n4, B=@n5)
   )
   Speed=AppendVector(A=@n6.PanX, B=@n6.PanX)
  ) {speedX=-0.5}
  n8 = AppendVector(A=@n6.ScaleX, B=@n6.ScaleX)
  n9 = Sine(None=ComponentMask(None=Multiply(A=@n7, B=@n8)) {r=true})
  n10 = Transform(None=VectorParameter() {group=07- WPO, parameterName=LocalWindStrentgth, sortPriority=1}.RGB) {transformSourceType=TRANSFORMSOURCE_Local}
  n11 = DynamicParameter() {parameterIndex=2, defaultValue=<rgba 0 0 0 0>, paramNames [SpeedX,SpeedY,SpeedZ,Bendiness]}
  n12 = MaterialFunctionCall(X=@n11.SpeedX, Y=@n11.SpeedX, Z=@n11.SpeedX) {materialFunction=&MakeFloat3}
  n13 = ScalarParameter() {group=07- WPO, parameterName=BendSmoothMin, sortPriority=11}
  n14 = ScalarParameter() {defaultValue=1, group=07- WPO, parameterName=BendSmoothMax, sortPriority=12}
  n15 = ComponentMask(None=NamedRerouteUsage()) {r=true}
  n16 = Add(
   A=Multiply(A=@n3, B=Multiply(A=@n9, B=@n10))
   B=Multiply(A=Multiply(A=@n12.Result, B=@n11.Bendiness), B=SmoothStep(Min=@n13, Max=@n14, Value=@n15))
  )
  n17 = ComponentMask(None=MaterialFunctionCall() {materialFunction=&LocalPosition}."Local Position (Excluding Offsets)") {g=true, r=true}
  n18 = ScalarParameter() {defaultValue=1000, group=07- WPO, parameterName=WorldNoiseSize}
  n19 = Time()
  n20 = ScalarParameter() {defaultValue=0.04, group=07- WPO, parameterName=WorldWindSpeed}
  n21 = TextureSampleParameter2D(UVs=Add(A=Divide(A=@n17, B=@n18), B=Multiply(A=@n19, B=@n20))) {group=07- WPO, parameterName=WindNoise, samplerType=SAMPLERTYPE_Masks, texture=/Game/Global/Textures/Default/Noise/WindNoise_Mask_1}
  n22 = ConstantBiasScale(None=@n21.RGB) {bias=-0.5, scale=2}
  n23 = Multiply(
   A=ScalarParameter() {defaultValue=5, group=07- WPO, parameterName=WorldWindStrength}
   B=Transform(None=Constant3Vector() {constant=<rgba 1 0 0 0>}) {transformSourceType=TRANSFORMSOURCE_Camera}
  )
  n24 = StaticSwitchParameter(True=Multiply(A=@n3, B=Multiply(A=@n22, B=@n23)), False=Constant()) {group=07- WPO, parameterName=AddWorldWind}
  n25 = ScalarParameter() {group=07- WPO, parameterName=SplineWidthBase}
  n26 = ScalarParameter() {group=07- WPO, parameterName=SplineWidthTip}
  n27 = NamedRerouteUsage()
  n28 = StaticBool()
  Add(
   A=Add(A=Multiply(A=@n1.Result, B=@n2.Camera Offset), B=Add(A=@n16, B=@n24))
   B=StaticSwitchParameter(
    True=MaterialFunctionCall(WidthBase=@n25, WidthTip=@n26, UVs for Projection=@n27, UVs for Thickness=@n27, UVs for texturing=@n27, Expand U or V UV Channel=@n28) {materialFunction=&SplineThicken}.WorldPositionOffset
    False=Constant()
   ) {group=07- WPO, parameterName=UseSplineThicken}
  )
  n29 = ScalarParameter() {defaultValue=0.4, group=03- Color, parameterName=InnerColorTransitionThreshold}
  n30 = Add(A=@n29, B=ScalarParameter() {defaultValue=0.7, group=03- Color, parameterName=InnerColorTransitionSoftness})
  n31 = ScalarParameter() {defaultValue=0.55, group=03- Color, parameterName=InnerColorBaseThreshold}
  n32 = NamedRerouteUsage()
  n33 = SmoothStep(
   Min=@n31
   Max=Add(A=@n31, B=ScalarParameter() {defaultValue=0.45, group=03- Color, parameterName=InnerColorBaseSoftness})
   Value=OneMinus(None=ComponentMask(None=@n32) {r=true})
  ) {constMin=0.55}
  n34 = NamedRerouteUsage()
  n35 = ComponentMask(None=@n34) {r=true}
  n36 = SmoothStep(Max=ScalarParameter() {defaultValue=0.2, group=04- Shape, parameterName=FlameBaseSoftness}, Value=@n35) {constMax=0.2}
  n37 = SmoothStep(Max=ScalarParameter() {defaultValue=0.5, group=04- Shape, parameterName=FlameTipSoftness}, Value=OneMinus(None=@n35)) {constMax=0.1}
  n38 = ScalarParameter() {defaultValue=0.4, group=04- Shape, parameterName=FlameSideSoftness}
  n39 = ComponentMask(None=@n34) {g=true}
  n40 = Multiply(A=@n39, B=OneMinus(None=@n39))
  n41 = StaticSwitchParameter(
   True=MaterialFunctionCall(
    Radius=ScalarParameter() {defaultValue=0.5, group=04- Shape, parameterName=SpriteRadius}
    Density=ScalarParameter() {defaultValue=2.33, group=04- Shape, parameterName=SpriteDensity}
   ) {materialFunction=&RadialGradientExponential}.RadialGradientExponential
   False=Multiply(A=Multiply(A=@n36, B=@n37), B=SmoothStep(Max=@n38, Value=@n40) {constMax=0.5})
  ) {group=04- Shape, parameterName=IsSpriteParticle}
  n42 = Add(
   A=DynamicParameter() {defaultValue=<rgba 0.5 0 0 0>, paramNames [UV1 U Offset, Dive, Camera Offset, HueShift]}.UV1 U Offset
   B=VertexColor().A
  )
  n43 = Constant()
  n44 = VectorParameter() {
   group=06- Distortion
   parameterName=DistoUV_Settings
   channelNames {a=PanY, b=PanX, g=ScaleY, r=ScaleX}
   defaultValue=<rgba 0.3 0.3 -2.2 0.2>
  }
  n45 = Panner(Coordinate=NamedRerouteUsage(), Speed=AppendVector(A=@n44.PanX, B=@n44.PanX)) {speedX=-0.5}
  n46 = AppendVector(A=@n44.ScaleX, B=@n44.ScaleX)
  n47 = TextureSampleParameter2D(UVs=Add(A=AppendVector(A=@n42, B=@n43), B=Multiply(A=@n45, B=@n46))) {
   group=06- Distortion
   parameterName=Distortion Texture3
   samplerType=SAMPLERTYPE_Grayscale
   texture=/Game/Effects/Fort_Effects/Textures/Fire/T_WispyFireNoise_Soft_M
  }
  n48 = Multiply(A=@n47.R, B=ScalarParameter() {defaultValue=0.25, group=04- Shape, parameterName=SubstractDisto}) {constB=0.25}
  n49 = Multiply(A=ComponentMask(None=Subtract(A=@n41, B=@n48)) {r=true}) {constB=3}
  n50 = NamedRerouteUsage()
  n51 = ScalarParameter() {defaultValue=0.2, group=06- Distortion, parameterName=Distortion Strength, sliderMax=1, sliderMin=-1}
  n52 = Constant()
  n53 = Panner(
   Coordinate=Add(A=NamedRerouteUsage(), B=Multiply(A=@n50, B=@n51) {constB=0.1})
   Speed=AppendVector(A=ScalarParameter() {defaultValue=-2, group=00- FlameShape, parameterName=BaseFlameSpeedX, sliderMax=5, sliderMin=-5}, B=@n52)
  ) {speedX=1}
  n54 = Constant() {r=1}
  n55 = AppendVector(A=ScalarParameter() {defaultValue=1, group=00- FlameShape, parameterName=Scale X, sliderMax=5}, B=@n54)
  n56 = ScalarParameter() {defaultValue=1, group=00- FlameShape, parameterName=Scale Y, sliderMax=5}
  n57 = Add(
   A=MaterialFunctionCall(
    UVs=Multiply(A=@n53, B=@n55)
    Texture Scale=MaterialFunctionCall(X=@n54, Y=@n56) {materialFunction=&MakeFloat2}.Result
   ) {materialFunction=&ScaleUVsByCenter}.UVs
   B=AppendVector(A=DynamicParameter() {defaultValue=<rgba 0.5 0 0 0>, paramNames [UV1 U Offset, Dive, Camera Offset, HueShift]}.UV1 U Offset, B=@n52)
  )
  n58 = TextureSampleParameter2D(UVs=LinearInterpolate(
    A=@n57
    B=MaterialFunctionCall(XY=@n57) {materialFunction=&Swizzle}.YX
    Alpha=ScalarParameter() {group=00- FlameShape, parameterName=SwizzleTrail, sliderMax=1}
   )) {group=04- Shape, parameterName=Mask Texture, texture=/Game/Effects/Fort_Effects/Textures/T_HornedJudgment_FlameTrail}
  n59 = OneMinus(None=ScalarParameter() {group=00- FlameShape, parameterName=Roundness})
  n60 = OneMinus(None=ComponentMask(None=NamedRerouteUsage()) {r=true})
  n61 = Multiply(
   A=LinearInterpolate(A=Multiply(A=@n49, B=@n58.R), B=@n49, Alpha=SmoothStep(Min=@n59, Value=@n60) {constMin=0.85})
   B=ScalarParameter() {defaultValue=1, group=01-Opacity, parameterName=FlameOpacity}
  )
  n62 = Saturate(None=Add(A=Power(Base=@n33) {constExponent=3}, B=@n61))
  n63 = ScalarParameter() {defaultValue=0.2, group=03- Color, parameterName=GradialColorTransitionThreshold}
  n64 = ScalarParameter() {defaultValue=0.4, group=03- Color, parameterName=GradialColorTransitionSoftness}
  n65 = ComponentMask(None=NamedRerouteUsage()) {r=true}
  n66 = Multiply(
   A=SmoothStep(Min=@n63, Max=Add(A=@n63, B=@n64), Value=OneMinus(None=@n65)) {constMax=0.7, constMin=0.3}
   B=LinearInterpolate(Alpha=SmoothStep(Min=@n29, Max=@n30, Value=@n62) {constMin=0.5}) {constA=0.5}
  )
  n67 = MaterialFunctionCall(
   1=VectorParameter() {group=03- Color, parameterName=OuterColorMult, sortPriority=0, defaultValue=<rgba 1 1 1 1>}.RGB
   2=VectorParameter() {group=03- Color, parameterName=MidColorMult, sortPriority=0, defaultValue=<rgba 1 1 1 1>}.RGB
   3=VectorParameter() {group=03- Color, parameterName=InnerColorMult, sortPriority=0, defaultValue=<rgba 1 1 1 1>}.RGB
   A=@n66
  ) {materialFunction=&Lerp_Multiple_Float3}
  n68 = ScalarParameter() {defaultValue=1, group=03- Color, parameterName=Emissive, sliderMax=2}
  n69 = ScalarParameter() {defaultValue=3, group=03- Color, parameterName=Emissive_Dive, sliderMax=3}
  n70 = DynamicParameter() {defaultValue=<rgba 0.5 0 0 0>, paramNames [UV1 U Offset, Dive, Camera Offset, HueShift]}
  n71 = Add(A=Multiply(A=@n66, B=LinearInterpolate(A=@n68, B=@n69, Alpha=@n70.Dive) {constA=1})) {constB=0.5}
  n72 = OneMinus(None=LinearInterpolate(
    A=ScalarParameter() {defaultValue=0.2, group=03- Color, parameterName=BaseEmissiveBoostSize}
    B=ScalarParameter() {defaultValue=0.3, group=03- Color, parameterName=DiveEmissiboostSize}
    Alpha=@n70.Dive
   ))
  n73 = NamedRerouteUsage()
  n74 = SmoothStep(
   Min=@n72
   Max=Add(A=@n72, B=ScalarParameter() {defaultValue=0.2, group=03- Color, parameterName=BaseEmissiveBoostFallof})
   Value=OneMinus(None=ComponentMask(None=@n73) {r=true})
  ) {constMin=0.7}
  n75 = ScalarParameter() {defaultValue=3, group=03- Color, parameterName=BaseEmissiveBoost}
  n76 = Multiply(A=Multiply(A=ParticleColor().RGB, B=Multiply(A=@n67.Lerp 3 Inputs, B=@n71)), B=Add(A=Multiply(A=@n74, B=@n75)))
  n77 = ScalarParameter() {defaultValue=0.55, parameterName=HueShift}
  n78 = DynamicParameter() {defaultValue=<rgba 0.5 0 0 0>, paramNames [UV1 U Offset, Dive, Camera Offset, HueShift]}
  n79 = ScalarParameter() {parameterName=UseHueShift}
  MaterialFunctionCall(Emissive=LinearInterpolate(
    A=@n76
    B=MaterialFunctionCall(Hue Shift Percentage=@n77, Texture=@n76) {materialFunction=&HueShift}.Result
    Alpha=Multiply(A=@n78.HueShift, B=@n79)
   )) {materialFunction=&MF_FX_EmissiveAdjustments}
  n80 = CurveAtlasRowParameter(CurveTime=@n66) {
   atlas=/Game/Athena/LinearColorCurves/CurveAtlas_Chainmail
   curve=/Game/Gadgets/Assets/VinderTech_GliderChute/Glider_GelatoCourage/FX/C_GelatoCourage_FlameColor
   defaultValue=10
   group=03- Color
   parameterName=Curve_Color
  }
  Multiply(A=@n80, B=@n71)
  Multiply(A=@n47.R, B=@n47.R)
  n81 = ScalarParameter() {defaultValue=0.1, group=05- Dissolve, parameterName=GradialDissolveStart}
  n82 = Add(A=@n81, B=ScalarParameter() {defaultValue=0.1, group=05- Dissolve, parameterName=GradialDissolvesoftness})
  n83 = NamedRerouteUsage()
  n84 = Multiply(A=NamedRerouteUsage(), B=ScalarParameter() {defaultValue=0.1, group=05- Dissolve, parameterName=GradialDissolveFlameShape})
  n85 = OneMinus(None=ComponentMask(None=Add(A=@n83, B=@n84)) {r=true})
  NamedRerouteDeclaration(None=LinearInterpolate(
    B=SmoothStep(Min=@n81, Max=@n82, Value=@n85) {constMax=0, constMin=0.5}
    Alpha=ScalarParameter() {defaultValue=1, group=05- Dissolve, parameterName=UseGradialDissolve}
   ) {constA=1}) {name=GradientDissolve, nodeColor=<rgba 1 0 0.9844 1>}
  n86 = DynamicParameter() {parameterIndex=1, defaultValue=<rgba 1 1 0 1>, paramNames [Param1,Param2,Dissolve,Param4]}
  n87 = Subtract(A=@n86.Dissolve, B=ScalarParameter() {defaultValue=0.5, group=05- Dissolve, parameterName=DissolveSoftness})
  NamedRerouteDeclaration(None=Saturate(None=SmoothStep(Min=@n87, Max=@n86.Dissolve, Value=@n47.R))) {name=Dissolve, nodeColor=<rgba 0 1 0.7422 1>}
  n88 = ComponentMask(None=NamedRerouteUsage()) {g=true}
  n89 = OneMinus(None=Multiply(A=@n88, B=OneMinus(None=@n88)))
  NamedRerouteDeclaration(None=Multiply(A=Subtract(A=@n47.R) {constB=0.3}, B=SmoothStep(Value=@n89) {constMin=0.4})) {name=Distortion, nodeColor=<rgba 1 0.3083 0.7353 1>}
  n90 = TextureCoordinate()
  n91 = MaterialFunctionCall(XY=MaterialFunctionCall(Vector or UVs=Add(A=@n90) {constB=-0.5}) {materialFunction=&VectorToRadialValue}.Radial Coordinates) {materialFunction=&Swizzle}
  n92 = StaticSwitchParameter(
   True=@n91.YX
   False=LinearInterpolate(A=@n90, B=TextureCoordinate() {coordinateIndex=1}, Alpha=ScalarParameter() {defaultValue=1, group=00- FlameShape, parameterName=Use_UV1})
  ) {group=04- Shape, parameterName=UseRadialUVs}
  n93 = MaterialFunctionCall(Float2=MaterialFunctionCall(XY=@n92) {materialFunction=&Swizzle}.YX) {materialFunction=&BreakOutFloat2Components}
  n94 = OneMinus(None=@n93.R)
  n95 = ScalarParameter() {group=00- FlameShape, parameterName=SwizzleUvs}
  NamedRerouteDeclaration(None=LinearInterpolate(
    A=@n92
    B=MaterialFunctionCall(X=@n94, Y=@n93.G) {materialFunction=&MakeFloat2}.Result
    Alpha=@n95
   )) {name=UV_1, nodeColor=<rgba 0.7812 1 0 1>}
  n96 = StaticSwitchParameter(True=@n91.YX, False=@n90) {group=04- Shape, parameterName=UseRadialUVs}
  n97 = MaterialFunctionCall(Float2=MaterialFunctionCall(XY=@n96) {materialFunction=&Swizzle}.YX) {materialFunction=&BreakOutFloat2Components}
  n98 = OneMinus(None=@n97.R)
  NamedRerouteDeclaration(None=LinearInterpolate(
    A=@n96
    B=MaterialFunctionCall(X=@n98, Y=@n97.G) {materialFunction=&MakeFloat2}.Result
    Alpha=@n95
   )) {name=UV_0, nodeColor=<rgba 0.8984 0 1 1>}
  n99 = Multiply(A=@n49, B=@n58.G)
  n100 = ScalarParameter() {defaultValue=1, group=01-Opacity, parameterName=GlowPow}
  n101 = Add(
   A=SmoothStep(
    Min=ScalarParameter() {group=00- FlameShape, parameterName=Opacity_TextureMin}
    Max=ScalarParameter() {defaultValue=0.4, group=00- FlameShape, parameterName=Opacity_TextureMax}
    Value=@n61
   ) {constMax=0.4}
   B=Multiply(A=Power(Base=@n99, Exp=@n100), B=ScalarParameter() {defaultValue=0.3, group=01-Opacity, parameterName=GlowStrength})
  )
  n102 = ScalarParameter() {defaultValue=10, group=01-Opacity, parameterName=Depth Fade Distance, sliderMax=100}
  n103 = DepthFade(Opacity=@n101, FadeDistance=LinearInterpolate(A=@n102, B=Multiply(A=@n102) {constB=0.5}, Alpha=@n61)) {fadeDistanceDefault=25}
  n104 = ScalarParameter() {defaultValue=100, group=01-Opacity, parameterName=Camera Fade Length, sliderMax=200}
  n105 = ScalarParameter() {defaultValue=25, group=01-Opacity, parameterName=Camera Fade Offset, sliderMax=200}
  n106 = Multiply(
   A=Multiply(
    A=ShadingPathSwitch(Default=@n103, Deferred=@n103, Forward=@n101, Mobile=@n101)
    B=MaterialFunctionCall(Fade Length=@n104, Fade Offset=@n105) {materialFunction=&CameraDepthFade}.Result
   )
   B=ScalarParameter() {defaultValue=1, group=01-Opacity, parameterName=Opacity, sliderMax=1}
  )
  n107 = NamedRerouteUsage()
  n108 = Multiply(A=Multiply(A=Multiply(A=@n106, B=@n107), B=NamedRerouteUsage()), B=ParticleColor().A)
  n109 = PixelNormalWS()
  n110 = VertexNormalWS()
  n111 = QualitySwitch(Default=@n109, Low=@n110, High=@n109, Medium=@n110, Epic=@n109)
  n112 = Abs(None=DotProduct(A=CameraVectorWS(), B=ShadingPathSwitch(Default=@n111, Deferred=@n111, Forward=@n110, Mobile=@n110)))
  n113 = OneMinus(None=@n112)
  n114 = VectorParameter() {
   group=Base
   parameterName=Mesh Fresnel Fade
   sortPriority=1
   channelNames {a="", b=Power, g=Invert, r=Use Mesh Fresnel Fade}
   defaultValue=<rgba 0 0 1 0>
  }
  n115 = Saturate(None=Power(Base=LinearInterpolate(A=@n112, B=@n113, Alpha=@n114.Invert), Exp=@n114.Power))
  Saturate(None=LinearInterpolate(A=@n108, B=Multiply(A=@n108, B=@n115), Alpha=@n114.Use Mesh Fresnel Fade))
  TextureCoordinate()
 ]
}
